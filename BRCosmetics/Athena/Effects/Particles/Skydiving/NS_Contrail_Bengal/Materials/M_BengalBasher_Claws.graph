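# Default Lit material for claw mesh particles in the Bengal Basher skydiving contrail Niagara effect.
# The graph holds a diffuse texture sample, a HueShift function driven by a slightly negative scalar parameter, and a linear interpolation.
# The blue channel of a packed claw maps texture drives the interpolation alpha.
# No links to the material outputs or the other lerp and hue shift inputs are shown.
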Material /BRCosmetics/Athena/Effects/Particles/Skydiving/NS_Contrail_Bengal/Materials/M_BengalBasher_Claws
graph {
 properties {
  bUsedWithNiagaraMeshParticles=true
  pixelDepthOffsetMode=PDOM_Legacy
  previewMesh=/BRCosmetics/Athena/Effects/Particles/Skydiving/NS_Contrail_Bengal/Meshes/sm_bengalBasher_Claws
  usedShadingModels=Default Lit
 }
 nodes [
  LinearInterpolate(Alpha=TextureSample() {texture=/BRCosmetics/Athena/Effects/Particles/Skydiving/NS_Contrail_Bengal/Materials/Textures/T_BengalBasher_ClawMaps}.B)
  MaterialFunctionCall(Hue Shift Percentage=ScalarParameter() {defaultValue=-0.06, parameterName=hue shift}) {materialFunction=&HueShift}
  TextureSample() {texture=/BRCosmetics/Athena/Effects/Particles/Skydiving/NS_Contrail_Bengal/Materials/Textures/T_BengalBasher_Diffuse}
 ]
}
