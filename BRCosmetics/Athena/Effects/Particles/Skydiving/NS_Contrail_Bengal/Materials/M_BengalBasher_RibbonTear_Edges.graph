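# Unlit translucent, two-sided material for the Bengal skydiving contrail's Niagara ribbons and meshes, apparently drawing the ragged edges of a tear.
# Panned cell noise breaks up sphere masks along UV V, one blending time-of-day color with mirrored screen-space claw-eye textures, another scaled by particle alpha.
# World-position-offset branches combine dynamic-parameter-driven trailing, Voronoi distortion with SplineThicken, normal-based noise, and a pendulum rotation with a previous-frame switch.
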Material /BRCosmetics/Athena/Effects/Particles/Skydiving/NS_Contrail_Bengal/Materials/M_BengalBasher_RibbonTear_Edges
graph {
 properties {
  bUsedWithNiagaraMeshParticles=true
  bUsedWithNiagaraRibbons=true
  blendMode=BLEND_Translucent
  pixelDepthOffsetMode=PDOM_Legacy
  shadingModel=MSM_Unlit
  twoSided=true
  usedShadingModels=Unlit
 }
 nodes [
  Add(B=VertexColor())
  n1 = DynamicParameter() {paramNames [Eye Fade, Y, Z, WPO Amount]}
  n2 = MaterialFunctionCall(X=@n1.Eye Fade, Y=@n1.Eye Fade, Z=@n1.Eye Fade) {materialFunction=&MakeFloat3}
  n3 = ObjectPositionWS()
  n4 = DynamicParameter() {parameterIndex=1, paramNames [X, Y, Z, WPO Amount]}
  n5 = OneMinus(None=ComponentMask(None=TextureCoordinate()) {g=true})
  n6 = ScalarParameter() {defaultValue=1.4, parameterName=Falloff}
  n7 = LinearInterpolate(
   B=Clamp(None=Subtract(A=@n2.Result, B=@n3), Min=Multiply(A=@n4.X) {constB=-1}, Max=@n4.X)
   Alpha=Multiply(A=Power(Base=@n5, Exp=@n6), B=@n1.WPO Amount)
  )
  n8 = VectorParameter() {parameterName=WPO Trail Multiplier, defaultValue=<rgba 1 1 1 0>}
  Add(A=Add(A=Multiply(A=@n7, B=@n8.RGB), B=NamedRerouteUsage()), B=NamedRerouteUsage())
  Add(B=VertexColor().R)
  Add(B=NamedRerouteUsage())
  Add(A=VectorParameter() {parameterName=Param_2}.RGB)
  LinearInterpolate(
   A=TransformPosition(None=ParticlePositionWS()) {transformSourceType=TRANSFORMPOSSOURCE_World, transformType=TRANSFORMPOSSOURCE_Instance}
   B=Multiply(B=ScalarParameter() {defaultValue=1, parameterName=bendScale}) {constB=0.25}
  )
  LinearInterpolate()
  n9 = MaterialFunctionCall() {materialFunction=&ScreenResolution}
  n10 = Multiply(
   A=Multiply(A=ScreenPosition().ViewportUV, B=Normalize(VectorInput=@n9.Visible Resolution))
   B=ScalarParameter() {defaultValue=20, parameterName=scale}
  )
  n11 = ScalarParameter() {defaultValue=-9.5, parameterName=Param_8, sliderMin=0.5}
  n12 = ScalarParameter() {defaultValue=-3, parameterName=Param_11}
  n13 = TextureObject() {texture=&/claw_eye_tex}
  n14 = TextureSample(UVs=Add(A=@n10, B=AppendVector(A=@n11, B=@n12)), Tex=@n13) {
   samplerSource=SSM_Clamp_WorldGroupSettings
   texture=&claw_eye_tex
  }
  n15 = MaterialFunctionCall(Float2=@n10) {materialFunction=&BreakOutFloat2Components}
  n16 = OneMinus(None=@n15.R)
  n17 = ScalarParameter() {defaultValue=8, parameterName=Right Eye X Offset, sliderMin=0.5}
  n18 = ScalarParameter() {defaultValue=-3, parameterName=Right Eye Y Offset}
  n19 = TextureSample(
   UVs=Add(
    A=MaterialFunctionCall(X=@n16, Y=@n15.G) {materialFunction=&MakeFloat2}.Result
    B=AppendVector(A=@n17, B=@n18)
   )
   Tex=@n13
  ) {
   samplerSource=SSM_Clamp_WorldGroupSettings
   texture=&claw_eye_tex
  }
  n20 = Add(A=Multiply(A=@n14.RGB, B=@n14.RGB), B=Multiply(A=@n19.RGB, B=@n19.RGB))
  n21 = DynamicParameter() {paramNames [Eye Fade, Y, Z, WPO Amount]}
  n22 = DynamicParameter() {parameterIndex=3, paramNames [Rand, Y, Z, WPO Amount]}
  n23 = Multiply(A=TextureCoordinate() {coordinateIndex=1}, B=Add(A=@n22.Rand, B=ScalarParameter() {defaultValue=4, parameterName=Param_3}))
  n24 = Time()
  n25 = Add(A=@n22.Rand, B=ScalarParameter() {defaultValue=3, parameterName=Time Scale})
  n26 = TextureSample(UVs=Panner(Coordinate=Add(A=@n23, B=@n22.Rand), Time=Multiply(A=@n24, B=@n25)) {speedX=-1}) {samplerType=SAMPLERTYPE_LinearGrayscale, texture=/Game/Effects/Fort_Effects/Textures/Tile/T_CellNoise_Tile_M}
  n27 = Multiply(A=@n26.G, B=ScalarParameter() {defaultValue=0.1})
  n28 = ComponentMask(None=TextureCoordinate()) {g=true}
  LinearInterpolate(
   A=MaterialFunctionCall() {materialFunction=/Game/Effects/Fort_Effects/Materials/_Functions/CurveColor_Time_of_Day_S23}.Result
   B=Multiply(A=Multiply(A=@n20, B=@n21.Eye Fade), B=ScalarParameter() {defaultValue=0.75, parameterName=Eye Intensity})
   Alpha=SphereMask(
    A=Add(A=@n27, B=@n28)
    B=ScalarParameter() {defaultValue=0.35, parameterName=Param_1}
    Radius=ScalarParameter() {defaultValue=0.15, parameterName=Param_2}
   )
  ) {constB=0}
  MaterialFunctionCall() {materialFunction=&Color_Time_of_Day}
  n29 = DynamicParameter() {parameterIndex=1, paramNames [X, Y, Z, WPO Amount]}
  n30 = ComponentMask(None=TransformPosition(None=MaterialFunctionCall(X=@n29.X, Y=@n29.X, Z=@n29.X) {materialFunction=&MakeFloat3}.Result) {transformSourceType=TRANSFORMPOSSOURCE_World, transformType=TRANSFORMPOSSOURCE_Camera}) {g=true, r=true}
  MaterialFunctionCall!(
   Vector2=Subtract(A=Normalize(VectorInput=@n30), B=ScalarParameter() {defaultValue=1, parameterName=Param_4})
   MaximumNumberOfDigits
   UVs
   DebugTextLocation RG_UpperRight BA_LowerLeft
   Component Spacing
  ) {materialFunction=/Engine/Functions/Engine_MaterialFunctions02/Utility/DebugFloat2Values}
  MaterialFunctionCall() {materialFunction=&BoundingBoxBased_0-1_UVW}
  n31 = Add(A=Multiply(A=@n26.G, B=ScalarParameter() {defaultValue=0.2, parameterName=Param_16}), B=TextureCoordinate())
  Multiply(
   A=ParticleColor().A
   B=SphereMask(
    A=ComponentMask(None=@n31) {g=true}
    B=ScalarParameter() {defaultValue=0.35, parameterName=Param_15}
    Radius=ScalarParameter() {defaultValue=0.306, parameterName=Param_12}
    Hardness=ScalarParameter() {defaultValue=1, parameterName=Param_17}
   )
  )
  n32 = OneMinus(None=ComponentMask(None=TextureCoordinate()) {g=true})
  n33 = WorldPosition()
  n34 = Time()
  n35 = ParticleRandom()
  n36 = Add(
   A=Floor(None=TransformPosition(None=@n33.XYZ) {transformSourceType=TRANSFORMPOSSOURCE_World, transformType=TRANSFORMPOSSOURCE_Instance})
   B=Multiply(A=Add(A=@n34, B=@n35), B=VectorParameter() {defaultValue=<rgba -700 0 0 0>}.RGB)
  )
  n37 = ScalarParameter() {defaultValue=0.001, parameterName=UV distortion scale}
  n38 = Multiply(
   A=TextureSample(UVs=Multiply(A=@n36, B=@n37)) {samplerType=SAMPLERTYPE_LinearColor, texture=/Game/Effects/Explosion_Update/Textures/VT_Voronoi_Seamless}.G
   B=ScalarParameter() {defaultValue=1, parameterName=UV distortion amount}
  )
  n39 = ScalarParameter() {defaultValue=9, parameterName=Base}
  n40 = TextureCoordinate()
  n41 = LinearInterpolate(
   A=ScalarParameter() {defaultValue=0.75, parameterName=Mid}
   B=ScalarParameter() {defaultValue=5, parameterName=Tip}
   Alpha=SphereMask(
    A=ComponentMask(None=@n40) {g=true}
    B=ScalarParameter() {defaultValue=0.8184, parameterName=Start}
    Radius=ScalarParameter() {defaultValue=0.756, parameterName=Rad}
    Hardness=ScalarParameter() {defaultValue=0.4037, parameterName=Hard}
   )
  )
  NamedRerouteDeclaration(None=Add(
    A=Multiply(A=@n32, B=@n38)
    B=MaterialFunctionCall(WidthBase=@n39, WidthTip=@n41) {materialFunction=&SplineThicken}.WorldPositionOffset
   )) {name=WPO, nodeColor=<rgba 0.125 1 0 1>}
  n42 = WorldPosition()
  n43 = DynamicParameter() {parameterIndex=2, paramNames [MeshOriginX,MeshOriginY,MeshOriginZ,MaskFalloff]}
  n44 = MaterialFunctionCall(R=@n43.MeshOriginX, G=@n43.MeshOriginX, B=@n43.MeshOriginX, A=@n43.MeshOriginX) {materialFunction=&AppendMany}
  n45 = DynamicParameter() {parameterIndex=3, paramNames [Rand, Y, Z, WPO Amount]}
  n46 = Subtract(
   A=MaterialFunctionCall(R=@n45.Rand, G=@n45.Rand, B=@n45.Rand) {materialFunction=&AppendMany}.RGB
   B=@n44.RGB
  )
  n47 = DynamicParameter() {parameterIndex=1, paramNames [X, Y, Z, WPO Amount]}
  n48 = Multiply(A=MaterialFunctionCall(R=@n47.X, G=@n47.X, B=@n47.X) {materialFunction=&AppendMany}.RGB) {constB=-1}
  n49 = RotateAboutAxis(
   NormalizedRotationAxis=CrossProduct(A=@n48, B=Normalize(VectorInput=@n46))
   RotationAngle=@n45.WPO Amount
   PivotPoint=@n44.RGB
   Position=@n42.XYZ
  )
  n50 = Subtract(A=ComponentMask(None=TextureCoordinate()) {g=true}, B=ScalarParameter() {parameterName=Mask Offset})
  n51 = Saturate(None=Power(Base=Saturate(None=@n50), Exp=@n43.MaskFalloff))
  n52 = DynamicParameter() {paramNames [Eye Fade, Y, Z, WPO Amount]}
  n53 = Subtract(
   A=MaterialFunctionCall(R=@n52.Eye Fade, G=@n52.Eye Fade, B=@n52.Eye Fade) {materialFunction=&AppendMany}.RGB
   B=@n44.RGB
  )
  n54 = RotateAboutAxis(
   NormalizedRotationAxis=CrossProduct(A=@n48, B=Normalize(VectorInput=@n53))
   RotationAngle=@n45.WPO Amount
   PivotPoint=@n44.RGB
   Position=@n42.XYZ
  )
  NamedRerouteDeclaration(None=PreviousFrameSwitch(Current Frame=LinearInterpolate(A=@n49, Alpha=@n51) {constB=0}, Previous Frame=LinearInterpolate(A=@n54, Alpha=@n51) {constB=0})) {name=Pendulum, nodeColor=<rgba 0.6641 0 1 1>}
  n55 = VertexNormalWS()
  n56 = TextureSample(UVs=Multiply(A=WorldPosition().XYZ, B=ScalarParameter() {defaultValue=0.005, parameterName=Noise Size})) {samplerType=SAMPLERTYPE_LinearColor, texture=/Game/Global/MaterialFunctions/VT_ClothSimWPO}
  NamedRerouteDeclaration(None=Multiply(A=Multiply(A=@n55, B=@n56.RGB), B=ScalarParameter() {defaultValue=1, parameterName=Noise Scale})) {name=Noise, nodeColor=<rgba 0.7109 1 0 1>}
  OneMinus()
  TransformPosition() {transformSourceType=TRANSFORMPOSSOURCE_World, transformType=TRANSFORMPOSSOURCE_Instance}
 ]
}
