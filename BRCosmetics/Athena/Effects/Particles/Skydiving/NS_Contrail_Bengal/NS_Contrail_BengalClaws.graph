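# Skydiving contrail Niagara system for the Bengal cosmetic, drawing claws, paper bits and tear trails from the player's hands.
# Claws pins oriented meshes to both hand bones, while Paper_Bits spawns drifting mesh and sprite debris there.
# Ribbon_Tears partitions particles into eight short-lived ribbons, four per hand, with array-driven offsets.
# ParaGlideLeanAlpha blends ribbon offsets, facing and debris behavior; bIsFrontEndPreview adds upward acceleration and a camera-dependent nudge to ribbons.
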NiagaraSystem /BRCosmetics/Athena/Effects/Particles/Skydiving/NS_Contrail_Bengal/NS_Contrail_BengalClaws
dependencies {
 usedDataInterfaces [
  &NiagaraDataInterfaceCurve
  &NiagaraDataInterfaceArrayPosition
  &NiagaraDataInterfaceArrayFloat
  &NiagaraDataInterfaceCamera
  &NiagaraDataInterfaceColorCurve
 ]
 usedDynamicInputs [&TransformPosition, &MakeFloatFromVector]
 usedModules [
  &self:SystemSpawnScript.SystemScriptSource.SystemScriptGraph.NiagaraNodeAssignment_0.SetVariables_#1
  &SystemState
  /Game/Effects/Niagara/Modules/Contrails/FN_Contrail_Setup
  &FN_Integrator
  &self:SystemSpawnScript.SystemScriptSource.SystemScriptGraph.NiagaraNodeAssignment_1.SetVariables_#2
 ]
}
emitters {
 Claws {
  emitterUpdateScript [
   &EmitterState(Life Cycle Mode=::NewEnumerator1, Loop Behavior=::NewEnumerator1, Loop Duration=1, UseLoopDelay=false)
   &SpawnRate(SpawnRate=50, Use Spawn Probability=false) {enabled=false}
   &SpawnBurst_Instantaneous(Use Spawn Probability=false, UseLoopCountLimit=false)
   &SpawnBurst_Instantaneous(Spawn Group=1, Use Spawn Probability=false, UseLoopCountLimit=false) {module=SpawnBurst_Instantaneous001}
  ]
  particleSpawnScript [
   &InitializeParticle(
    Color Mode=::NewEnumerator0
    Lifetime=5
    Mesh Scale=<0.4 0.3 0.3>
    Mesh Scale Mode=::NewEnumerator4
    Sprite Size Mode=::NewEnumerator3
    UsePositionOffset=true
    Write Material Random=true
   )
   &PartitionParticles()
  ]
  particleUpdateScript [
   &ParticleState(Kill Particles When Lifetime Has Elapsed=false)
   &SkeletalMeshLocation(
    Bone / Socket Index=$Particles.PartitionParticles.Partition
    Bone Sampling Mode=::NewEnumerator4
    MaskSpawnGroup=false
    Orientation Sampling=::NewEnumerator1
    Sampled Position Offset=&MaskVectorBySpawnGroup(Masked Vector=<13 4 0>, Passthrough Vector=<-13 -4 0>)
    Sampled Scale Coordinate Space:ENiagaraCoordinateSpace=::World
    Skeletal Mesh=DataInterface() {propertyValues {
      DefaultMesh=null
      ExcludeBoneName=root
      OptionalComponentClass=null
      PreviewMesh=null
      SkinningMode=SkinOnTheFly
      SoftSourceActor=null
      SourceMode=Default
      UvSetIndex=0
      WholeMeshLOD=-1
      bExcludeBone=false
      bReadDeformedGeometry=true
      bRequireCurrentFrameData=true
      ComponentTags []
      FilteredBones [hand_L,hand_r]
      FilteredSockets []
      MeshUserParameter {parameter {name=None, typeDefHandle {registeredTypeIndex=1}, varData []}}
      SamplingRegions []
     }}
   )
   &UpdateMeshOrientation(
    Facing Direction=&RotateVectorByQuaternion(
     Vector To Rotate=&MaskVectorBySpawnGroup(Masked Vector=<1 0 0>, Passthrough Vector=<-1 0 0>)
     Quaternion=$Particles.SkeletalMeshLocation.SampledRotation
    )
    Facing Interpolation=false
    Orientation Method=::NewEnumerator1
    Override Quat=false
    TransformFacing=false
    TransformUp=false
    Up Direction:Vector3f=&RotateVectorByQuaternion(
     Vector To Rotate=&MaskVectorBySpawnGroup(Masked Vector=<0 1 0>, Passthrough Vector=<0 -1 0>)
     Quaternion=$Particles.SkeletalMeshLocation.SampledRotation
    )
   )
  ]
  rendererClasses [&NiagaraMeshRendererProperties]
  renderers [
   {rendererClass=&NiagaraMeshRendererProperties, rendererIndex=0}
   {rendererClass=&NiagaraMeshRendererProperties, rendererIndex=1}
  ]
 }
 Empty {
  bEnabled=false
  emitterUpdateScript [
   &EmitterState()
   &SpawnBurst_Instantaneous(Use Spawn Probability=false, UseLoopCountLimit=false)
  ]
  particleSpawnScript [&InitializeParticle(Color Mode=::NewEnumerator0, Lifetime=5, UsePositionOffset=false, Write Material Random=true)]
  particleUpdateScript [&ParticleState()]
  rendererClasses [&NiagaraMeshRendererProperties]
  renderers [{rendererClass=&NiagaraMeshRendererProperties, rendererIndex=0}]
 }
 Paper_Bits {
  emitterUpdateScript [
   &EmitterState(Life Cycle Mode=::NewEnumerator1, Loop Duration=1, UseLoopDelay=false)
   &SpawnRate(SpawnRate=&Lerp_Float(A=90, B=150, Alpha=$User.ParaGlideLeanAlpha), Use Spawn Probability=false)
  ]
  particleSpawnScript [
   &PartitionParticles()
   &InitializeParticle(
    Color=<rgba 500 500 500 1>
    Lifetime Max=0.6
    Lifetime Min=0.3
    Lifetime Mode=::NewEnumerator1
    Mesh Scale=<0.4 0.3 0.3>
    Mesh Scale Mode=::NewEnumerator4
    Sprite Size Mode=::NewEnumerator1
    Uniform Sprite Size Min=2
    UsePositionOffset=true
    Write Material Random=true
   )
   &SkeletalMeshLocation(
    Bone / Socket Index=$Particles.PartitionParticles.Partition
    Bone Sampling Mode=::NewEnumerator4
    Inherited Velocity Speed Limit=3500
    MaskSpawnGroup=false
    Orientation Sampling=::NewEnumerator1
    Sampled Position Offset=&MaskVectorBySpawnGroup(Masked Vector=<13 0 0>, Passthrough Vector=<-13 0 0>, Spawn Group Mask=$Particles.PartitionParticles.Partition)
    Sampled Scale Coordinate Space:ENiagaraCoordinateSpace=::World
    Skeletal Mesh=DataInterface() {propertyValues {
      DefaultMesh=null
      ExcludeBoneName=root
      OptionalComponentClass=null
      PreviewMesh=null
      SkinningMode=SkinOnTheFly
      SoftSourceActor=null
      SourceMode=Default
      UvSetIndex=0
      WholeMeshLOD=-1
      bExcludeBone=false
      bReadDeformedGeometry=true
      bRequireCurrentFrameData=true
      ComponentTags []
      FilteredBones [hand_L,hand_r]
      FilteredSockets []
      MeshUserParameter {parameter {name=None, typeDefHandle {registeredTypeIndex=1}, varData []}}
      SamplingRegions []
     }}
    Use Sampled Mesh Attribute as Ribbon ID=false
    Velocity Sampling=::NewEnumerator0
    VelocitySpeedLimit=true
   )
   &ShapeLocation(MaskSpawnGroup=false, Non Uniform Scale=<1 0 0>, Override Seed=false, Sphere Radius=10, UseOwnerScale=true)
   &AddVelocity(Velocity=&RandomRangeVector(Minimum=<-50 -50 0>, Maximum=<50 50 250>))
  ]
  particleUpdateScript [
   &ParticleState()
   &ScaleSpriteSize(Uniform Curve Sprite Scale=DataInterface() {propertyValues {Curve {keys [{interpMode=RCIM_Cubic, tangentMode=RCTM_Break, time=0, value=1}, {tangentMode=RCTM_User, time=1, value=0}]}}})
   &ScaleColor(ScaleA=true, ScaleRGB=true)
   &AccelerationForce() {enabled=false}
   &Drag(
    Drag=&FloatFromCurve(FloatCurve=DataInterface() {propertyValues {Curve {keys [{interpMode=RCIM_Cubic, leaveTangent=8.6345, tangentMode=RCTM_Break, time=0, value=1}, <at 1 3>]}}})
    Use Linear Drag=true
    Use Rotational Drag=true
   )
   &InheritSourceMovement(
    Applied Position Falloff Distance=750
    Applied Position Scale=&Lerp_Vector(A=<-0.1 -0.1 -0.12>, B=<0 0 -0.24>, Alpha=$User.ParaGlideLeanAlpha)
    ApplyForce=false
    ApplyPosition=true
    ApplyPositionFalloff=false
   )
   &SolveForcesAndVelocity(Clamp Velocity=false, Limit Acceleration=false)
  ]
  rendererClasses [&NiagaraMeshRendererProperties, &NiagaraSpriteRendererProperties]
  renderers [
   {rendererClass=&NiagaraMeshRendererProperties, rendererIndex=0}
   {rendererClass=&NiagaraSpriteRendererProperties, rendererIndex=1}
  ]
 }
 Ribbon_Tears {
  emitterUpdateScript [
   &EmitterState(Life Cycle Mode=::NewEnumerator1, Loop Duration=1, Scalability Mode=::NewEnumerator1, UseLoopDelay=false)
   &SpawnRate(SpawnRate=2500, Use Spawn Probability=false)
   &SpawnPerUnit(Module.Use Max Movement Threshold=true, Module.Use Movement Tolerance=false, Spawn Spacing=0.75, Use Spawn Probability=false) {enabled=false}
   &SpawnBurst_Instantaneous(Spawn Count=8, Use Spawn Probability=false, UseLoopCountLimit=false) {enabled=false}
  ]
  particleSpawnScript [
   &PartitionParticles(Number Of Partitions=8, Write Partition to Ribbon ID=::NewEnumerator0)
   /SetVariables(Particles.Partition Group:NiagaraInt32=&MakeCustomIntFromBool(
     Bool=&SetBoolByIntComparison(A=$Particles.PartitionParticles.Partition, B=3, Comparison Type=::NewEnumerator0)
     True INT=1
     False INT=0
    )) {isSetParameters=true}
   /SetVariables(Particles.Position Array=DataInterface() {propertyValues {
      GpuSyncMode=SyncCpuToGpu
      MaxElements=0
      PositionData [<16 4 2>, <16 4 -4>, <16 4 6.5>, <16 4 -11>, <-16 -4 1.5>, <-15 -4 7>, <-16 -4 -4>, <-15 -4 -9.5>]
     }}) {isSetParameters=true}
   /SetVariables(Particles.Dive Position Array=DataInterface() {propertyValues {
      GpuSyncMode=SyncCpuToGpu
      MaxElements=0
      PositionData [<16 4 2>, <16 4 -3>, <16 4 6.5>, <16 4 -11>, <-16 -4 4>, <-15 -4 11>, <-16 -4 -2>, <-15 -4 -8>]
     }}) {isSetParameters=true}
   &InitializeParticle(
    Color Mode=::NewEnumerator0
    Lifetime=0.1
    Ribbon Facing Mode=::NewEnumerator1
    Ribbon Facing Vector:Vector3f=&TransformVector(
     Vector=&Lerp_Vector(A=<1 0 0>, B=<-1 1 0>, Alpha=$User.ParaGlideLeanAlpha)
     Destination Space=::Simulation
     Apply Scale=true
    )
    Ribbon Width=7
    Ribbon Width Mode=::NewEnumerator1
    Sprite Size Mode=::NewEnumerator3
    Uniform Sprite Size=5
    UsePositionOffset=false
    Write Material Random=true
   )
   &SkeletalMeshLocation(
    Bone / Socket Index=$Particles.Partition Group
    Bone Sampling Mode=::NewEnumerator4
    MaskSpawnGroup=false
    Orientation Sampling=::NewEnumerator1
    Sampled Position Offset=&Lerp_Vector(
     A=&Multiply_Vector(
      A=/Niagara/DynamicInputs/Arrays/SelectPositionFromArray(
       Array Sampling Mode:ENiagaraArraySamplingMode=::NewEnumerator1
       Direct Array Index:NiagaraInt32=$Particles.PartitionParticles.Partition
       Interpolation:NiagaraFloat=?
       Direct Array Mode:ENiagara_WrapClamp=::NewEnumerator0
       Evaluation Type:ENiagaraRandomnessEvaluation=::NewEnumerator0
       Normalized Interpolation:NiagaraBool=true
       Recalculate Random Each Loop:NiagaraBool=false
       Position Array:NiagaraDataInterfaceArrayPosition=$Particles.Position Array
       Randomness Mode:ENiagaraRandomnessMode=::NewEnumerator0
       Fixed Random Seed:NiagaraBool=false
       Random Seed:NiagaraInt32=?
      )
      B=<1 1 1>
     )
     B=&Multiply_Vector(
      A=&SelectPositionFromArray(
       Array Sampling Mode:ENiagaraArraySamplingMode=::NewEnumerator1
       Direct Array Index:NiagaraInt32=$Particles.PartitionParticles.Partition
       Interpolation:NiagaraFloat=?
       Direct Array Mode:ENiagara_WrapClamp=::NewEnumerator0
       Evaluation Type:ENiagaraRandomnessEvaluation=::NewEnumerator0
       Normalized Interpolation:NiagaraBool=true
       Recalculate Random Each Loop:NiagaraBool=false
       Position Array:NiagaraDataInterfaceArrayPosition=$Particles.Dive Position Array
       Randomness Mode:ENiagaraRandomnessMode=::NewEnumerator0
       Fixed Random Seed:NiagaraBool=false
       Random Seed:NiagaraInt32=?
      )
      B=<1 1 1>
     )
     Alpha=$User.ParaGlideLeanAlpha
    )
    Sampled Scale Coordinate Space:ENiagaraCoordinateSpace=::World
    Skeletal Mesh=DataInterface() {propertyValues {
      DefaultMesh=null
      ExcludeBoneName=root
      OptionalComponentClass=null
      PreviewMesh=null
      SkinningMode=SkinOnTheFly
      SoftSourceActor=null
      SourceMode=Default
      UvSetIndex=0
      WholeMeshLOD=-1
      bExcludeBone=false
      bReadDeformedGeometry=true
      bRequireCurrentFrameData=true
      ComponentTags []
      FilteredBones [hand_L,hand_r]
      FilteredSockets []
      MeshUserParameter {parameter {name=None, typeDefHandle {registeredTypeIndex=1}, varData []}}
      SamplingRegions []
     }}
    Use Sampled Mesh Attribute as Ribbon ID=false
   )
   /SetVariables(Particles.Float Array=DataInterface() {propertyValues {GpuSyncMode=SyncCpuToGpu, MaxElements=0, FloatData [0.15,0.16,0.1,0.11,0.13,0.1,0.13,0.1]}}) {isSetParameters=true}
   /SetVariables(Particles.Lifetime=&SelectFloatFromArray(
     Direct Array Index=&self:ScratchDynamicInput_01()
     Float Selection Array=$Particles.Float Array
    )) {isSetParameters=true}
   &InheritVelocity(Inherited Velocity Amount Scale=<1 1 0.17>, Inherited Velocity Speed Limit=1750, Speed Limit=true, UseSpeedThreshold=false)
   /SetVariables(Particles.Position=&MakeCustomVectorFromBool(
     Bool=$User.bIsFrontEndPreview
     True Vector=&AddVectorToPosition(
      Position=$Particles.Position
      Vector=&Lerp_Vector(
       A=<0 0 0>
       B=<0 -3 0>
       Alpha=&ClampFloat(Float=&DotProduct(
         Vector A=$Engine.Owner.SystemYAxis
         Vector B=&ReturnCameraVector(
          Camera Vectors=::NewEnumerator4
          Coordinate Space=::Simulation
          Origin Position=$Particles.Position
          CameraQuery=DataInterface() {propertyValues {bRequireCurrentFrameData=false}}
         )
        ))
      )
     )
     False Vector=Linked() {linkedVariable {name=Particles.Position, type=NiagaraPosition}}
    )) {isSetParameters=true}
  ]
  particleUpdateScript [
   &ParticleState()
   &ScaleRibbonWidth(Ribbon Width Scale=&FloatFromCurve(FloatCurve=DataInterface() {propertyValues {Curve {keys [
         {interpMode=RCIM_Cubic, leaveTangent=11.8414, tangentMode=RCTM_Break, time=0, value=0.9}
         {interpMode=RCIM_Cubic, tangentMode=RCTM_Break, time=0.0447, value=1.25}
         <at 1 0>
        ]}}}))
   &ScaleColor(
    Linear Color Curve=DataInterface() {propertyValues {
      AlphaCurve {keys [<at 0.49 1>,<at 1 0>]}
      BlueCurve {keys [<at 0 1>,<at 1 1>]}
      GreenCurve {keys [<at 0 1>,<at 1 1>]}
      RedCurve {keys [<at 0 1>,<at 1 1>]}
     }}
    Scale Mode=::NewEnumerator2
   )
   &DynamicMaterialParameters(
    Index 0 Param 1=1
    Index 3 Param 1=&Divide_Float(A=&self:ScratchDynamicInput(), B=8)
    Module.Index0Param0Write=true
    Module.Index0Param1Write=true
    Module.Index0Param2Write=true
    Module.Index0Param3Write=true
    Module.Index3Param0Write=true
    Module.Index3Param1Write=true
    Module.Index3Param2Write=true
    Module.Index3Param3Write=true
    Write Parameter Index 3=true
   )
   &AccelerationForce(Acceleration=&Multiply_VectorByFloat(
     Vector=<0 0 15000>
     Float=&MakeCustomFloatFromBool(Bool=$User.bIsFrontEndPreview, True Float=1, False Float=0)
    ))
   &SolveForcesAndVelocity(Clamp Velocity=false, Limit Acceleration=false)
  ]
  rendererClasses [&NiagaraSpriteRendererProperties, &NiagaraRibbonRendererProperties]
  renderers [
   {rendererClass=&NiagaraSpriteRendererProperties}
   {rendererClass=&NiagaraRibbonRendererProperties, rendererIndex=1}
  ]
 }
}
summary {
 systemName=NS_Contrail_BengalClaws
 emitters [
  {emitterName=Claws, rendererClasses [&NiagaraMeshRendererProperties]}
  {emitterName=Paper_Bits, rendererClasses [&NiagaraMeshRendererProperties, &NiagaraSpriteRendererProperties]}
  {emitterName=Ribbon_Tears, rendererClasses [&NiagaraSpriteRendererProperties, &NiagaraRibbonRendererProperties]}
  {bEnabled=false, emitterName=Empty, rendererClasses [&NiagaraMeshRendererProperties]}
 ]
 userVariables [
  User.bIsFrontEnd {type=NiagaraBool}
  User.bIsFrontEndPreview {type=NiagaraBool, defaultValue=true}
  User.ParaGlideLeanAlpha {type=NiagaraFloat, defaultValue=0, description=""}
  User.RibbonIDOffset {type=NiagaraInt32}
  User.SkeletalMesh {
   type=NiagaraDataInterfaceSkeletalMesh
   defaultValue {
    struct=NiagaraExt_VariableValue_DataInterface
    value {dataInterface=&self:NiagaraDataInterfaceSkeletalMesh_0}
   }
  }
 ]
}
